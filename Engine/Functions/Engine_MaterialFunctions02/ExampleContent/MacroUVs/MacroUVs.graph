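# MacroUVs is an example material that builds camera-aligned UVs spanning a whole object, for sampling a lighting map texture.
# Object-relative world position is dotted with transformed view X and Y axes, divided by object radius and offset by half to center the texture.
# Leftover branches experiment with camera-distance scaling, scene depth and screen position.
# The graph shows no connections to material outputs, including from the texture sample.
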Material /Engine/Functions/Engine_MaterialFunctions02/ExampleContent/MacroUVs/MacroUVs
graph {
 properties {pixelDepthOffsetMode=PDOM_Legacy, usedShadingModels=Default Lit}
 nodes [
  n1 = Subtract(A=WorldPosition().XYZ, B=ObjectPositionWS())
  n2 = Constant3Vector() {constant=<rgba 1 0 0 0>}
  n3 = Constant3Vector() {constant=<rgba 0 1 0 0>}
  n4 = AppendVector(
   A=DotProduct(A=@n1, B=Transform(None=@n2) {transformSourceType=TRANSFORMSOURCE_View})
   B=DotProduct(A=@n1, B=Transform(None=@n3) {transformSourceType=TRANSFORMSOURCE_View})
  )
  n5 = Divide(A=@n4, B=Distance(A=WorldPosition().XYZ, B=CameraPositionWS()))
  n6 = ObjectRadius()
  Multiply(A=@n5, B=@n6)
  Multiply(A=@n5) {constB=3}
  PixelDepth()
  ScalarParameter() {defaultValue=1}
  n7 = Constant3Vector() {constant=<rgba 0 0 1 0>}
  Subtract(A=SceneDepth().R, B=DotProduct(A=@n1, B=Transform(None=@n7) {transformSourceType=TRANSFORMSOURCE_View}))
  TextureSample(UVs=Add(A=Divide(A=@n4, B=@n6)) {constB=0.5}) {texture=/Engine/Functions/Engine_MaterialFunctions02/ExampleContent/Textures/newhplaneLightingMap}
  Transform(None=ScreenPosition().ViewportUV) {transformSourceType=TRANSFORMSOURCE_View}
 ]
}
